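# Orange emissive material for the gondola of the Artemis blimp set, a base Material despite its MI_ prefix.
# It uses Default Lit shading and is flagged for Nanite, skeletal mesh and mesh deformer use.
# Its nodes are constant colors and one vector parameter with an over-bright orange default, likely the glow.
# No links between the nodes or to material outputs appear.
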Material /Game/Environments/Artemis/Sets/Blimp/Materials/MI_Blimp_Gondala_Emissive_Orange
graph {
 properties {
  bUsedWithMeshDeformer=true
  bUsedWithNanite=true
  bUsedWithSkeletalMesh=true
  pixelDepthOffsetMode=PDOM_Legacy
  usedShadingModels=Default Lit
 }
 nodes [
  Constant3Vector() {constant=<rgba 0 0 1 1>}
  Constant3Vector() {constant=<rgba 0.6514 0.1946 0.0452 1>}
  Constant3Vector() {constant=<rgba 1 1 0.998 1>}
  Constant3Vector() {constant=<rgba 0.3594 0.3594 0.3587 1>}
  VectorParameter() {defaultValue=<rgba 20 6.6667 0 1>}
 ]
}
