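# Base material for the interior trims of the Blimp set in the Artemis environment, despite its MI_ instance prefix.
# It holds three plain texture samples for the _D diffuse, _N normal and _S mask maps, with no parameters or math nodes.
# It uses Default Lit shading, is flagged for Nanite and assigns the Metal physical material.
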Material /Game/Environments/Artemis/Sets/Blimp/Materials/MI_Blimp_InteriorTrims
graph {
 properties {bUsedWithNanite=true, physMaterial=/Game/FortressPhysicalMaterials/Metal, pixelDepthOffsetMode=PDOM_Legacy, usedShadingModels=Default Lit}
 nodes [
  TextureSample() {samplerType=SAMPLERTYPE_Masks, texture=/Game/Environments/Artemis/Sets/Blimp/Textures/T_Blimp_InteriorTrims_S}
  TextureSample() {texture=/Game/Environments/Artemis/Sets/Blimp/Textures/T_Blimp_InteriorTrims_D}
  TextureSample() {samplerType=SAMPLERTYPE_Normal, texture=/Game/Environments/Artemis/Sets/Blimp/Textures/T_Blimp_InteriorTrims_N}
 ]
}
